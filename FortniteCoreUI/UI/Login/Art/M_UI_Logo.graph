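# UI-domain translucent material for a login screen logo, with the image supplied by the Logo texture object parameter.
# It reads the texture's own size through a named reroute and passes it, with normalized UI UVs, to MF_UI_UVScaling.
# The Scaled to Fit output drives a clamped texture sample, so the logo's UV fitting follows whatever texture is assigned.
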Material /FortniteCoreUI/UI/Login/Art/M_UI_Logo
graph {
 properties {
  bUsedWithStaticMesh=false
  blendMode=BLEND_Translucent
  materialDomain=MD_UI
  pixelDepthOffsetMode=PDOM_Legacy
  usedShadingModels=Default Lit
 }
 nodes [
  n1 = TextureObjectParameter() {parameterName=Logo}
  NamedRerouteDeclaration(None=TextureProperty(Tex=@n1)) {name=Texture Size, nodeColor=<rgba 0 0 0 1>}
  TextureSample(
   UVs=MaterialFunctionCall(
    UV=MaterialFunctionCall() {materialFunction=&GetUserInterfaceUV}.Normalized UV
    TextureSize=NamedRerouteUsage()
   ) {materialFunction=&MF_UI_UVScaling}.Scaled to Fit
   Tex=@n1
  ) {samplerSource=SSM_Clamp_WorldGroupSettings}
 ]
}
