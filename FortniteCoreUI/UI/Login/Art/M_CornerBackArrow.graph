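# Translucent UI material for the login screen that draws a back-arrow icon on an angled corner backdrop, both with drop shadows.
# Rotated UVs through a narrow SmoothStep cut the straight-edged backdrop, and AAThreshold turns the arrow texture into a crisp, resizable shape.
# Shared offset parameters shift the UVs for both shadows.
# Color layers through lerps from shadow to icon, while opacity takes the maximum of all four masks.
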Material /FortniteCoreUI/UI/Login/Art/M_CornerBackArrow
graph {
 properties {
  bUsedWithStaticMesh=false
  blendMode=BLEND_Translucent
  materialDomain=MD_UI
  pixelDepthOffsetMode=PDOM_Legacy
  usedShadingModels=Default Lit
 }
 nodes [
  n1 = VectorParameter() {parameterName=ShadowColor, defaultValue=<rgba 0.0052 0.0052 0.0056 0.45>}
  n2 = TextureCoordinate()
  n3 = Constant2Vector() {r=-0.2}
  n4 = AppendVector(A=ScalarParameter() {defaultValue=0.035, parameterName=ShadowOffsetX}, B=ScalarParameter() {defaultValue=0.05, parameterName=ShadowOffsetY})
  n5 = Constant2Vector()
  n6 = ScalarParameter() {defaultValue=0.8762, parameterName=BGRotation, sliderMin=-0.2}
  n7 = MaterialFunctionCall(UVs=Add(A=@n2, B=Subtract(A=@n3, B=@n4)), Rotation Center=@n5, "Rotation Angle (0-1)"=@n6) {materialFunction=&CustomRotator}
  n8 = Constant() {r=0.5}
  n9 = OneMinus(None=MaterialFunctionCall(Alpha=ComponentMask(None=@n7.Rotated Values) {r=true}, Min=Subtract(A=@n8) {constB=0.0035}, Max=@n8) {materialFunction=&SmoothStep}.Result)
  n10 = LinearInterpolate(B=@n1.RGB, Alpha=@n9)
  n11 = VectorParameter() {parameterName=BGColor, defaultValue=<rgba 0.0423 0.1119 0.3763 1>}
  n12 = MaterialFunctionCall(UVs=Add(A=@n2, B=@n3), Rotation Center=@n5, "Rotation Angle (0-1)"=@n6) {materialFunction=&CustomRotator}
  n13 = Constant() {r=0.5}
  n14 = OneMinus(None=MaterialFunctionCall(Alpha=ComponentMask(None=@n12.Rotated Values) {r=true}, Min=Subtract(A=@n13) {constB=0.0035}, Max=@n13) {materialFunction=&SmoothStep}.Result)
  n15 = VectorParameter() {parameterName=ShadowColor, defaultValue=<rgba 0.0052 0.0052 0.0056 0.45>}
  n16 = ScalarParameter() {defaultValue=0.175, parameterName=ShapeSize}
  n17 = Subtract(A=TextureCoordinate(), B=@n4)
  n18 = Multiply(
   A=MaterialFunctionCall!(
    Threshold=@n16
    Gradient=TextureSample(UVs=@n17) {samplerSource=SSM_Clamp_WorldGroupSettings, texture=&/T-Icon-BackArrowk-64}.R
    Filter Width Override
    Filter Width Fudge Factor
   ) {materialFunction=/Game/UI/Materials/AAThreshold}.Result
   B=@n15.A
  )
  n19 = MaterialFunctionCall(
   Threshold=@n16
   Gradient=TextureSample() {samplerSource=SSM_Clamp_WorldGroupSettings, texture=&T-Icon-BackArrowk-64}.R
  ) {materialFunction=&AAThreshold}
  LinearInterpolate(
   A=LinearInterpolate(A=LinearInterpolate(A=@n10, B=@n11.RGB, Alpha=@n14), B=@n15.RGB, Alpha=@n18)
   B=VectorParameter() {parameterName=IconColor, defaultValue=<rgba 1 1 1 1>}.RGB
   Alpha=@n19.Result
  )
  Max(A=Max(A=Max(A=@n14, B=@n19.Result), B=Multiply(A=@n9, B=@n1.A) {constB=0.75}), B=@n18)
 ]
}
