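# M_EpicLogo_DF is a translucent UI material for the Epic logo in the login screen art.
# It feeds the red channel of the T_EpicLogo_DF texture, apparently a distance field, into the AAThreshold function to cut an anti-aliased logo shape.
# A ShapeSize parameter sets the threshold, and an IconColor parameter supplies a tint that defaults to white.
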Material /FortniteCoreUI/UI/Login/Art/M_EpicLogo_DF
graph {
 properties {
  bUsedWithStaticMesh=false
  blendMode=BLEND_Translucent
  materialDomain=MD_UI
  pixelDepthOffsetMode=PDOM_Legacy
  usedShadingModels=Default Lit
 }
 nodes [
  MaterialFunctionCall!(
   Threshold=ScalarParameter() {defaultValue=0.18, parameterName=ShapeSize}
   Gradient=TextureSample() {texture=&/T_EpicLogo_DF}.R
   Filter Width Override
   Filter Width Fudge Factor
  ) {materialFunction=/Game/UI/Materials/AAThreshold}
  VectorParameter() {parameterName=IconColor, defaultValue=<rgba 1 1 1 1>}
 ]
}
